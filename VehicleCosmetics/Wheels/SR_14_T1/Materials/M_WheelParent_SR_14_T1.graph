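# Translucent parent material for SR_14_T1 vehicle wheels, adding an animated electric glow effect to areas masked by vertex color blue.
# Two panning cloud noises distort scrolling gradient samples into streaks, which drive tinted glow color and opacity.
# These blend over MF_WheelParent_Simple's base color, emissive and otherwise solid opacity, then pass through vehicle damage and de-rez transition functions.
# An LOD static switch changes the streak UV tiling.
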Material /VehicleCosmetics/Wheels/SR_14_T1/Materials/M_WheelParent_SR_14_T1
graph {
 properties {
  bAutomaticallySetUsageInEditor=false
  bUseMaterialAttributes=true
  bUsedWithMeshDeformer=true
  bUsedWithSkeletalMesh=true
  blendMode=BLEND_Translucent
  pixelDepthOffsetMode=PDOM_Legacy
  previewMesh=/VehicleCosmetics/Wheels/SR_14_T2/Meshes/SK_Wheel_SR_14_T2
  translucencyLightingMode=TLM_Surface
  usedShadingModels=Default Lit
 }
 nodes [
  Constant() {r=1}
  Constant() {r=1}
  n1 = GetMaterialAttributes(None=MaterialFunctionCall() {materialFunction=&MF_WheelParent_Simple}.MaterialAttributes) {attributeGetTypes [#1, #2, #3]}
  n2 = VertexColor()
  n3 = LinearInterpolate(A=@n1.Emissive Color, B=NamedRerouteUsage(), Alpha=@n2.B)
  n4 = LinearInterpolate(A=@n1.Base Color, B=NamedRerouteUsage(), Alpha=@n2.B)
  n5 = NamedRerouteUsage()
  n6 = LinearInterpolate(A=Constant() {r=1}, B=@n5, Alpha=@n2.B)
  MaterialFunctionCall(In=MaterialFunctionCall(Material=SetMaterialAttributes!(MaterialAttributes=@n1.MaterialAttributes, Emissive Color=@n3, Base Color=@n4, Opacity=@n6) {attributeSetTypes [#1, #2, #3]}) {materialFunction=&MF_BaseVehicle_Damage}.Result) {materialFunction=&MF_VehicleDeRez_Transition}
  n7 = Time()
  n8 = ScalarParameter() {defaultValue=1, group=5_Special Effects, parameterName=PanningSpeed}
  n9 = Multiply(A=Time(), B=ScalarParameter() {defaultValue=1, group=5_Special Effects, parameterName=ElectricitySpeed})
  n10 = TextureSample(UVs=Panner(Coordinate=TextureCoordinate() {uTiling=0.4, vTiling=0.4}, Time=@n9) {speedX=-0.2, speedY=-0.2}) {texture=/VehicleCosmetics/Wheels/_SharedTextures/FX/DCU_CloudDistort_Skin_T}
  n11 = TextureSample(UVs=Panner(Coordinate=TextureCoordinate() {uTiling=0.2, vTiling=0.2}, Time=@n9) {speedX=0.2, speedY=0.2}) {texture=&DCU_CloudDistort_Skin_T}
  n12 = Clamp(None=ConstantBiasScale(None=Add(A=@n10.R, B=@n11.R)) {bias=-0.5, scale=1})
  n13 = Add(
   A=Panner(Coordinate=TextureCoordinate() {uTiling=0.5, vTiling=0.5}, Time=Multiply(A=@n7, B=@n8)) {speedX=0.5, speedY=0.4}
   B=Multiply(A=Constant() {r=0.3}, B=@n12)
  )
  n14 = Clamp(None=ConstantBiasScale(None=TextureSample(UVs=@n13) {texture=/VehicleCosmetics/Wheels/SR_14_T8/Textures/wheel_sr_14_Gradiant_V}.R) {bias=-0.5, scale=3})
  n15 = Add(
   A=Multiply(A=@n14, B=Constant() {r=0.5})
   B=Multiply(A=@n12, B=ScalarParameter() {defaultValue=0.2, group=5_Special Effects, parameterName=ElectricityIntencity})
  )
  n16 = Add(
   A=@n15
   B=MaterialFunctionCall(UVs=TextureCoordinate(), "Rotation Angle (0-1)"=Constant() {r=-0.25}) {materialFunction=&CustomRotator}.Rotated Values
  )
  n17 = Add(
   A=@n15
   B=MaterialFunctionCall(UVs=TextureCoordinate() {uTiling=2, vTiling=5}, "Rotation Angle (0-1)"=Constant() {r=-0.25}) {materialFunction=&CustomRotator}.Rotated Values
  )
  n18 = Multiply(
   A=OneMinus(None=@n14)
   B=StaticSwitchParameter(
    True=TextureSample(UVs=@n16) {texture=&wheel_sr_14_Gradiant_V}.R
    False=TextureSample(UVs=@n17) {texture=&wheel_sr_14_Gradiant_V}.R
   ) {group=5_Special Effects, parameterName=LOD}
  )
  n19 = Clamp(None=ConstantBiasScale(None=@n18) {bias=-0.4, scale=5})
  n20 = ScalarParameter() {defaultValue=0.3, group=5_Special Effects, parameterName=WhiteCenter_Intensity}
  n21 = Multiply(
   A=Add(
    A=Multiply(A=@n19, B=@n20)
    B=VectorParameter() {group=5_Special Effects, parameterName=GlowColor, defaultValue=<rgba 0.4233 0.5271 0.5149 1>}.RGB
   )
   B=ScalarParameter() {defaultValue=1.5, group=5_Special Effects, parameterName=GlowIntensity}
  )
  NamedRerouteDeclaration(None=@n21) {name=Emissive, nodeColor=<rgba 1 0.75 0 1>}
  NamedRerouteDeclaration(None=@n21) {name=BaseColor, nodeColor=<rgba 0.1745 1 0.42 1>}
  OneMinus(None=@n2.R)
  n22 = TextureSampleParameter2D() {
   group=5_Special Effects
   parameterName=LOD_Alpha
   samplerType=SAMPLERTYPE_LinearColor
   texture=/VehicleCosmetics/Wheels/SR_14_T1/Textures/T_SR_14_T1_Alpha_LOD
  }
  StaticSwitchParameter(
   True=LinearInterpolate(A=Constant(), B=NamedRerouteDeclaration(None=@n18) {name=Opacity, nodeColor=<rgba 0.1745 1 0.42 1>}, Alpha=@n22.R)
   False=@n5
  ) {group=5_Special Effects, parameterName=LOD}
 ]
}
