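# Vehicle cosmetic wheel material for SR_14_T1 that adds a Silver TreadRank sign and fresnel glow to the MF_WheelParent_Simple base.
# Vertex color red masks changes to base color, normal and emissive, with sign textures sampled on UV1 and a static switch replacing their opacity with a constant.
# Two MF_Fresnel_Advanced edge terms sum into a tinted, intensity-scaled glow.
# The attributes then pass through MF_BaseVehicle_Damage and MF_VehicleDeRez_Transition.
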Material /VehicleCosmetics/Wheels/SR_14_T1/Materials/M_Wheel_SR_14_T1
graph {
 properties {
  bAutomaticallySetUsageInEditor=false
  bUseMaterialAttributes=true
  bUsedWithMeshDeformer=true
  bUsedWithSkeletalMesh=true
  pixelDepthOffsetMode=PDOM_Legacy
  previewMesh=/VehicleCosmetics/Wheels/SR_14_T1/Meshes/SK_Wheel_SR_14_T1
  usedShadingModels=Default Lit
 }
 nodes [
  n1 = TextureCoordinate() {coordinateIndex=1}
  n2 = StaticSwitchParameter(
   True=Constant()
   False=TextureSampleParameter2D(UVs=@n1) {parameterName=Sign_Opacity, texture=/VehicleCosmetics/Wheels/SR_14_T1/Textures/T_SR_14_T1_TreadRank_Silver_RGB}.R
  ) {parameterName=Sign}
  n3 = LinearInterpolate(
   A=NamedRerouteUsage()
   B=TextureSampleParameter2D(UVs=@n1) {parameterName=Sign_BC, texture=/VehicleCosmetics/Wheels/SR_14_T1/Textures/T_SR_14_T1_TreadRank_Silver_D}.RGB
   Alpha=@n2
  )
  n4 = MaterialFunctionCall() {materialFunction=&MF_WheelParent_Simple}
  LinearInterpolate(
   A=TextureSample() {samplerType=SAMPLERTYPE_Normal, texture=/VehicleCosmetics/Wheels/SR_14_T1/Textures/T_SR_14_T1_Tire_N}.RGB
   B=@n3
   Alpha=@n4.VertexColorR
  )
  n5 = GetMaterialAttributes(None=@n4.MaterialAttributes) {attributeGetTypes [#1, #2, #3]}
  n6 = TextureSampleParameter2D(UVs=@n1) {
   group=5_Special Effects
   parameterName=SIgn_N
   samplerType=SAMPLERTYPE_Normal
   texture=/VehicleCosmetics/Wheels/SR_14_T1/Textures/T_SR_14_T1_TreadRank_Silver_N
  }
  n7 = LinearInterpolate(A=@n5.Normal, B=LinearInterpolate(A=NamedRerouteUsage(), B=@n6.RGB, Alpha=@n2), Alpha=@n4.VertexColorR)
  n8 = LinearInterpolate(A=@n5.Base Color, B=@n3, Alpha=@n4.VertexColorR)
  n9 = LinearInterpolate(A=NamedRerouteUsage(), B=NamedRerouteUsage(), Alpha=@n4.VertexColorR)
  MaterialFunctionCall(In=MaterialFunctionCall(Material=SetMaterialAttributes!(MaterialAttributes=@n5.MaterialAttributes, Normal=@n7, Base Color=@n8, Emissive Color=@n9) {attributeSetTypes [#1, #2, #3]}) {materialFunction=&MF_BaseVehicle_Damage}.Result) {materialFunction=&MF_VehicleDeRez_Transition}
  NamedRerouteDeclaration(None=@n7) {name=Normal_FR, nodeColor=<rgba 1 0 0.8438 1>}
  n10 = ScalarParameter() {defaultValue=1.5, group=5_Special Effects, parameterName=FresnelGlow_NoNorm_Sharpness, sortPriority=33}
  n11 = Add(
   A=Multiply(
    A=MaterialFunctionCall!(Input Normal, Input Sharpness=@n10, Input Two Sided) {materialFunction=/VehicleCosmetics/Wheels/CarCar_Holographic/Materials/MF_Fresnel_Advanced}.Edge
    B=ScalarParameter() {defaultValue=1, group=5_Special Effects, parameterName=FresnelGlow_NoNorm_Addin, sortPriority=33}
   )
   B=MaterialFunctionCall(
    Input Normal=NamedRerouteUsage()
    Input Sharpness=ScalarParameter() {defaultValue=1, group=5_Special Effects, parameterName=FresnelGlow_Sharpness, sortPriority=33}
   ) {materialFunction=&MF_Fresnel_Advanced}.Edge
  )
  n12 = VectorParameter() {group=5_Special Effects, parameterName=GlowColor, sortPriority=33, defaultValue=<rgba 0.9387 0 1 1>}
  NamedRerouteDeclaration(None=Multiply(
    A=Multiply(A=@n11, B=@n12.RGB)
    B=ScalarParameter() {defaultValue=1.5, group=5_Special Effects, parameterName=FresnelGlow_Intensity, sortPriority=33}
   )) {name=Emissive, nodeColor=<rgba 0.9735 1 0 1>}
  NamedRerouteDeclaration(None=@n5.Normal) {name=Normal_Tire, nodeColor=<rgba 1 0.2344 0 1>}
  NamedRerouteDeclaration(None=@n5.Base Color) {name=BaseColor, nodeColor=<rgba 0 0.4141 1 1>}
  NamedRerouteDeclaration(None=@n5.Emissive Color) {name=Emissive_Source, nodeColor=<rgba 0 1 0.6953 1>}
 ]
}
